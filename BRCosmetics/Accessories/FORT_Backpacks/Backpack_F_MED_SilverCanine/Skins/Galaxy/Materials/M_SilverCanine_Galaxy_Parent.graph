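# Parent material for the Galaxy skin of the SilverCanine backpack, built on the 2023 character-parent PreFX/PostFX functions.
# It chains Ice and Galaxy crew-style functions and holographic glitter, with an isGold switch adding carbon fiber, including its clear-coat normal.
# A gemstone layer runs bump-offset triangle normals, skipped on low quality tiers, through thin-film and crystal functions.
# A switch chooses channel-mask or emissive-derived alpha for the final attribute blend before PostFX.
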Material /BRCosmetics/Accessories/FORT_Backpacks/Backpack_F_MED_SilverCanine/Skins/Galaxy/Materials/M_SilverCanine_Galaxy_Parent
graph {
 properties {
  bUseMaterialAttributes=true
  bUsedWithMeshDeformer=true
  bUsedWithMorphTargets=true
  bUsedWithSkeletalMesh=true
  pixelDepthOffsetMode=PDOM_Legacy
  shadingModel=MSM_FromMaterialExpression
  subsurfaceProfile=/Game/Packages/Fortress_SharedMaterials/SubSurfaceProfile/Character_SubsurfaceProfile
  usedShadingModels=Default Lit | Subsurface | Clear Coat | Subsurface Profile | Cloth
 }
 nodes [
  n1 = NamedRerouteUsage()
  n2 = MaterialFunctionCall!(
   Tech Art Mask=NamedRerouteUsage()
   Mask=NamedRerouteUsage()
   In=MaterialFunctionCall!(TechArt=NamedRerouteUsage(), Fresnel=NamedRerouteUsage(), In=NamedRerouteUsage()) {materialFunction=/Game/Global/Materials/Character_Parent_2023/CrewStyle_Parent_2023/Ice/MF_CrewStyle_Ice_Inst}.Result
  ) {materialFunction=/Game/Global/Materials/Character_Parent_2023/CrewStyle_Parent_2023/Galaxy/MF_CrewStyle_Galaxy_Inst}
  n3 = MaterialFunctionCall!(Alpha=NamedRerouteUsage(), In=@n2.Result) {materialFunction=/Game/Global/MaterialFunctions/TechArt/MF_CarbonFiberDualLayer}
  n4 = MaterialFunctionCall(
   BaseNormal=NamedRerouteUsage()
   AdditionalNormal=TextureSample(UVs=@n3.CarbonFiberUvs) {samplerType=SAMPLERTYPE_Normal, texture=/Game/Weapons/WeaponSkins/Wraps/Textures/T_Wrap_CarbonFiber_N}.RGB
  ) {materialFunction=&BlendAngleCorrectedNormals}
  ClearCoatNormalCustomOutput!(None=StaticSwitchParameter(True=LinearInterpolate(A=@n1, B=@n4.Result, Alpha=@n3.Mask), False=NamedRerouteUsage()) {parameterName=isGold})
  n5 = Multiply(
   A=StaticComponentMaskParameter(None=NamedRerouteUsage()) {defaultB=true, group=SliverCabineGemstone, parameterName=GemstoneChannel}
   B=ScalarParameter() {defaultValue=0.6, group=SliverCabineGemstone, parameterName=GemstoneBlend}
  ) {constB=0.6}
  n6 = TextureSampleParameter2D() {
   group=06_Emissive
   parameterName=Emissive
   samplerSource=SSM_Wrap_WorldGroupSettings
   sortPriority=1
   texture=&T_BlacksRGB
  }
  n7 = Max(A=@n6.R, B=@n6.R)
  n8 = Saturate(None=Multiply(A=Max(A=@n7, B=@n6.B)) {constB=50})
  n9 = MaterialFunctionCall() {materialFunction=&MF_CharacterParent_2023_PreFX}
  MaterialFunctionCall(
   Material=BlendMaterialAttributes(
    A=NamedRerouteUsage()
    B=NamedRerouteUsage()
    Alpha=StaticSwitchParameter(True=@n5, False=@n8) {defaultValue=true, group=SliverCabineGemstone, parameterName=GemstoneMaskMethod}
   )
   M=@n9.M
   Pre Skinned Local Position=@n9.M
   Distance Blend=@n9.M
   Fresnel=@n9.M
  ) {materialFunction=&MF_CharacterParent_2023_PostFX}
  NamedRerouteDeclaration(None=@n9.M) {name=Mask, nodeColor=<rgba 0 1 1 1>}
  NamedRerouteDeclaration(None=@n9.Tech Art Mask) {name=TechArtMask, nodeColor=<rgba 1 0.6797 0 1>}
  NamedRerouteDeclaration(None=@n9.Fresnel) {name=Fresnel, nodeColor=<rgba 0 1 0.4375 1>}
  NamedRerouteDeclaration(None=MaterialFunctionCall!(Alpha=NamedRerouteUsage(), In=StaticSwitchParameter(True=@n3.Result, False=@n2.Result) {parameterName=isGold}, Fresnel=NamedRerouteUsage()) {materialFunction=/Game/Global/MaterialFunctions/TechArt/MF_GlitterHolographic}.Result) {name=InputMA, nodeColor=<rgba 0 0.9297 1 1>}
  n10 = GetMaterialAttributes(None=NamedRerouteUsage()) {attributeGetTypes [
    #1
    #2
    #3
    #4
    #5
   ]}
  n11 = BumpOffset(Coordinate=Multiply(A=TextureCoordinate(), B=Constant() {r=-2.97}), Height=Constant() {r=-75.2159}) {heightRatio=0.01}
  n12 = MaterialFunctionCall(
   BaseNormal=@n10.Normal
   AdditionalNormal=MaterialFunctionCall(
    Normal=TextureSample(UVs=@n11) {samplerType=SAMPLERTYPE_Normal, texture=/Game/Global/Materials/MaterialGym/Materials/TechArtPatterns/Textures/T_Triangles_N}.RGB
    Flatness=ScalarParameter() {defaultValue=0.55, group=SliverCabineGemstone, parameterName=DetailsNormalFlatness}
   ) {materialFunction=&FlattenNormal}.Result
  ) {materialFunction=&BlendAngleCorrectedNormals}
  n13 = MakeMaterialAttributes!(
   BaseColor=@n10.Base Color
   Metallic=@n10.Base Color
   Specular=@n10.Base Color
   Roughness=@n10.Base Color
   Anisotropy
   EmissiveColor
   Opacity
   OpacityMask
   Normal=MaterialFunctionCall!(
    Low _ PC Low and Med=@n10.Normal
    "High _ XB1, PS4"=@n12.Result
    "Epic _ PC High, XSX, PS5"=@n12.Result
    --------
    "Mobile Low _ PC Perf, Android"=@n10.Normal
    Mobile Med _ Switch=@n10.Normal
   ) {materialFunction=/Game/Global/MaterialFunctions/TechArt/MF_Fortnite_QualitySwitch_Vector3}.Result
   Tangent
   WorldPositionOffset
   SubsurfaceColor
   ClearCoat
   ClearCoatRoughness
   AmbientOcclusion
   Refraction
   CustomizedUVs_0
   CustomizedUVs_1
   CustomizedUVs_2
   CustomizedUVs_3
   CustomizedUVs_4
   CustomizedUVs_5
   CustomizedUVs_6
   CustomizedUVs_7
   PixelDepthOffset
   ShadingModel
   Displacement
  )
  n14 = Multiply(A=Fresnel(ExponentIn=Constant() {r=1}) {baseReflectFraction=0}, B=Constant() {r=0.74})
  n15 = ScalarParameter() {defaultValue=5, group=SliverCabineGemstone, parameterName=ThinFlimAlpha}
  NamedRerouteDeclaration(None=MaterialFunctionCall!(Material=MaterialFunctionCall!(Material Attributes=@n13, Fresnel=@n14, Alpha=@n15) {materialFunction=/BRCosmetics/Accessories/FORT_Backpacks/Backpack_F_MED_SilverCanine/Materials/MF_ThinFilm_SilverCanine}.Result) {materialFunction=/BRCosmetics/Accessories/FORT_Backpacks/Backpack_F_MED_SilverCanine/Materials/MF_Crystal_SilverCanine}.Result) {name=Gemstone, nodeColor=<rgba 1 0 0.2344 1>}
  NamedRerouteDeclaration(None=@n9.Result) {name=MA, nodeColor=<rgba 0 1 0.2031 1>}
  NamedRerouteDeclaration(None=GetMaterialAttributes(None=@n3.Result) {attributeGetTypes [#5]}.Normal) {name=InputNormal, nodeColor=<rgba 0.5703 1 0 1>}
 ]
}
